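# Parental controls sub-screen that tells an opted-out user they can click Continue to re-enable Parental Controls.
# A centered vertical stack holds a styled rich-text notice above a fixed-width Continue button, which also triggers from the GenericFaceButtonLeft input action.
# The button is a widget inherited from native parent FortParentalControls_AskToReEnable, and the event graph holds only unconnected stub events, so this Blueprint supplies just layout and styling.
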WidgetBlueprint /FortniteCoreUI/UI/ContentControls/ParentalSubScreen_AskToReEnable
graphs [EventGraph {nodes [
   Event PreConstruct()
   Event Construct()
   Event Tick()
  ]}]
variables []
widgets {
 info {
  inheritedWidgetCount=1
  parentClass=/Script/FortniteUI.FortParentalControls_AskToReEnable
  rootWidgetClass=&Overlay
  widgetCount=6
 }
 widgets [
  SubScreen_AskToReEnable {widgetClassPath=&Overlay, bIsVariable=true, widget=SubScreen_AskToReEnable}
  VerticalBox_7 {
   widgetClassPath=&VerticalBox
   parent=SubScreen_AskToReEnable
   widget=VerticalBox_7
   slot {class=OverlaySlot, properties {horizontalAlignment=HAlign_Fill, verticalAlignment=VAlign_Center}}
  }
  EnableDescription_2 {
   widgetClassPath=&CommonRichTextBlock
   parent=VerticalBox_7
   widget=EnableDescription_2
   properties {
    autoWrapText=true
    bOverrideDefaultStyle=true
    justification=Center
    text="You have currently opted out of <text color=\"00FFD6FF\" fontface=\"black\" case=\"upper\" size=\"30\">Parental Controls</>. Click Continue to Re-Enable them."
    defaultTextStyleOverride {
     overflowPolicy=Clip
     transformPolicy=None
     colorAndOpacity {colorUseRule=UseColor_Specified, specifiedColor=<rgba 1 1 1 1>}
     font {
      bForceMonospaced=false
      bMaterialIsStencil=false
      fontMaterial=null
      fontObject=&BurbankSmall
      letterSpacing=0
      monospacedWidth=1
      size=28
      skewAmount=0
      typefaceFontName=Bold
      outlineSettings {
       bApplyOutlineToDropShadows=false
       bMiteredCorners=false
       bSeparateFillAlpha=false
       outlineMaterial=null
       outlineSize=0
       outlineColor=<rgba 0 0 0 1>
      }
     }
     highlightColor {colorUseRule=UseColor_Specified, specifiedColor=<rgba 0 0 0 1>}
     highlightShape {
      bIsDynamicallyLoaded=false
      drawAs=Image
      imageType=NoImage
      mirroring=NoMirror
      resourceName=None
      resourceObject=null
      tiling=NoTile
      imageSize=<32 32>
      margin=<ltrb 0 0 0 0>
      outlineSettings {
       bUseBrushTransparency=false
       roundingType=HalfHeightRadius
       width=0
       color {colorUseRule=UseColor_Specified, specifiedColor=<rgba 0 0 0 0>}
       cornerRadii=<0 0 0 0>
      }
      tintColor {colorUseRule=UseColor_Specified, specifiedColor=<rgba 1 1 1 1>}
      uVRegion {bIsValid=false, max=<0 0>, min=<0 0>}
     }
     selectedBackgroundColor {colorUseRule=UseColor_Specified, specifiedColor=<rgba 1 1 1 1>}
     shadowColorAndOpacity=<rgba 0 0 0 1>
     shadowOffset=<0 0>
     strikeBrush {
      bIsDynamicallyLoaded=false
      drawAs=Image
      imageType=NoImage
      mirroring=NoMirror
      resourceName=None
      resourceObject=null
      tiling=NoTile
      imageSize=<32 32>
      margin=<ltrb 0 0 0 0>
      outlineSettings {
       bUseBrushTransparency=false
       roundingType=HalfHeightRadius
       width=0
       color {colorUseRule=UseColor_Specified, specifiedColor=<rgba 0 0 0 0>}
       cornerRadii=<0 0 0 0>
      }
      tintColor {colorUseRule=UseColor_Specified, specifiedColor=<rgba 1 1 1 1>}
      uVRegion {bIsValid=false, max=<0 0>, min=<0 0>}
     }
     underlineBrush {
      bIsDynamicallyLoaded=false
      drawAs=Image
      imageType=NoImage
      mirroring=NoMirror
      resourceName=None
      resourceObject=null
      tiling=NoTile
      imageSize=<32 32>
      margin=<ltrb 0 0 0 0>
      outlineSettings {
       bUseBrushTransparency=false
       roundingType=HalfHeightRadius
       width=0
       color {colorUseRule=UseColor_Specified, specifiedColor=<rgba 0 0 0 0>}
       cornerRadii=<0 0 0 0>
      }
      tintColor {colorUseRule=UseColor_Specified, specifiedColor=<rgba 1 1 1 1>}
      uVRegion {bIsValid=false, max=<0 0>, min=<0 0>}
     }
    }
   }
   slot {class=VerticalBoxSlot}
  }
  Spacer_104 {widgetClassPath=&Spacer, parent=VerticalBox_7, widget=Spacer_104, properties {size=<1 48>}, slot {class=VerticalBoxSlot}}
  SizeBox_8 {
   widgetClassPath=&SizeBox
   parent=VerticalBox_7
   widget=SizeBox_8
   properties {bOverride_WidthOverride=true, widthOverride=800}
   slot {class=VerticalBoxSlot, properties {horizontalAlignment=HAlign_Center, verticalAlignment=VAlign_Top}}
  }
  Button_AskToReEnable {
   widgetClassPath=&IconTextButton
   bInherited=true
   parent=SizeBox_8
   widget=Button_AskToReEnable
   properties {
    buttonText=Continue
    onButtonBaseClicked=(&self:WidgetTree.Button_AskToReEnable.HandleOnButtonClicked)
    onButtonBaseDoubleClicked=(&self:WidgetTree.Button_AskToReEnable.HandleOnButtonDoubleClicked)
    onButtonBaseHovered=(&self:WidgetTree.Button_AskToReEnable.HandleOnButtonHovered)
    onButtonBaseUnhovered=(&self:WidgetTree.Button_AskToReEnable.HandleOnButtonUnhovered)
    onSelectedChangedBase=(&self:WidgetTree.Button_AskToReEnable.HandleOnSelectedChanged)
    style=/FortUILibrary/Foundation/Buttons/ButtonStyles/Skew/ButtonStyle-Skew_Desirable.ButtonStyle-Skew_Desirable_C
    triggeringInputAction {dataTable=&InputActionDataTable, rowName=GenericFaceButtonLeft}
   }
   slot {class=SizeBoxSlot}
  }
 ]
}
