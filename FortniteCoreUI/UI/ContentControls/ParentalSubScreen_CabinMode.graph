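# Parental controls sub-screen widget for Cabin Mode email, based on FortParentalControls_CabinModeEmailSubScreen.
# A vertically centered column stacks a "COMPLETE ACCOUNT SETUP" title, a spacer and the inherited CabinEmailManagementWidget.
# On initialization it binds to input method changes and runs the handler once with the current input type.
# The handler branches on whether input is Touch, but neither branch leads to any action.
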WidgetBlueprint /FortniteCoreUI/UI/ContentControls/ParentalSubScreen_CabinMode
graphs [EventGraph {nodes [
   n1 = OnInputMethodChanged!() -> (OutputDelegate, then, bNewInputType:ECommonInputType Enum) {is=title K2Node_CustomEvent}
   > Branch(Condition=Select!(MouseAndKeyboard:bool=false, Gamepad:bool=false, Touch:bool=true, Count:bool=false, Index:ECommonInputType Enum=@n1.bNewInputType) -> (ReturnValue:bool))
   Event OnInitialized()
   > Bind Event to On Input Method Changed(self=Get CommonInputSubsystem(), Delegate=@n1)
   > OnInputMethodChanged!(execute, self, bNewInputType:ECommonInputType Enum=GetCurrentInputType(self=Get CommonInputSubsystem())) -> (then) {is=title K2Node_CallFunction}
  ]}]
variables []
widgets {
 info {
  inheritedWidgetCount=1
  parentClass=/Script/FortniteUI.FortParentalControls_CabinModeEmailSubScreen
  rootWidgetClass=&Overlay
  widgetCount=5
 }
 widgets [
  SubScreen_DisableControls {widgetClassPath=&Overlay, bIsVariable=true, widget=SubScreen_DisableControls}
  VerticalBox_6 {
   widgetClassPath=&VerticalBox
   parent=SubScreen_DisableControls
   widget=VerticalBox_6
   slot {class=OverlaySlot, properties {horizontalAlignment=HAlign_Fill, verticalAlignment=VAlign_Center}}
  }
  Text_Title {
   widgetClassPath=&CommonRichTextBlock
   parent=VerticalBox_6
   widget=Text_Title
   properties {
    bIsScrollingEnabled=false
    bOverrideDefaultStyle=true
    defaultTextStyleOverrideClass=/FortUILibrary/Styles/Typography/_TextStyle_SubH1._TextStyle_SubH1_C
    text=COMPLETE ACCOUNT SETUP
    textTransformPolicy=ToUpper
    wrapTextAt=900
    defaultTextStyleOverride {
     overflowPolicy=Clip
     transformPolicy=None
     colorAndOpacity {colorUseRule=UseColor_Specified, specifiedColor=<rgba 0.7656 0.7656 0.7656 1>}
     font {
      bForceMonospaced=false
      bMaterialIsStencil=false
      fontMaterial=null
      fontObject=&BurbankSmall
      letterSpacing=0
      monospacedWidth=1
      size=28
      skewAmount=0
      typefaceFontName=Bold
      outlineSettings {
       bApplyOutlineToDropShadows=false
       bMiteredCorners=false
       bSeparateFillAlpha=false
       outlineMaterial=null
       outlineSize=0
       outlineColor=<rgba 0 0 0 1>
      }
     }
     highlightColor {colorUseRule=UseColor_Specified, specifiedColor=<rgba 0 0 0 1>}
     highlightShape {
      bIsDynamicallyLoaded=false
      drawAs=Image
      imageType=NoImage
      mirroring=NoMirror
      resourceName=None
      resourceObject=null
      tiling=NoTile
      imageSize=<32 32>
      margin=<ltrb 0 0 0 0>
      outlineSettings {
       bUseBrushTransparency=false
       roundingType=HalfHeightRadius
       width=0
       color {colorUseRule=UseColor_Specified, specifiedColor=<rgba 0 0 0 0>}
       cornerRadii=<0 0 0 0>
      }
      tintColor {colorUseRule=UseColor_Specified, specifiedColor=<rgba 1 1 1 1>}
      uVRegion {bIsValid=false, max=<0 0>, min=<0 0>}
     }
     selectedBackgroundColor {colorUseRule=UseColor_Specified, specifiedColor=<rgba 1 1 1 1>}
     shadowColorAndOpacity=<rgba 0 0 0 1>
     shadowOffset=<0 0>
     strikeBrush {
      bIsDynamicallyLoaded=false
      drawAs=Image
      imageType=NoImage
      mirroring=NoMirror
      resourceName=None
      resourceObject=null
      tiling=NoTile
      imageSize=<32 32>
      margin=<ltrb 0 0 0 0>
      outlineSettings {
       bUseBrushTransparency=false
       roundingType=HalfHeightRadius
       width=0
       color {colorUseRule=UseColor_Specified, specifiedColor=<rgba 0 0 0 0>}
       cornerRadii=<0 0 0 0>
      }
      tintColor {colorUseRule=UseColor_Specified, specifiedColor=<rgba 1 1 1 1>}
      uVRegion {bIsValid=false, max=<0 0>, min=<0 0>}
     }
     underlineBrush {
      bIsDynamicallyLoaded=false
      drawAs=Image
      imageType=NoImage
      mirroring=NoMirror
      resourceName=None
      resourceObject=null
      tiling=NoTile
      imageSize=<32 32>
      margin=<ltrb 0 0 0 0>
      outlineSettings {
       bUseBrushTransparency=false
       roundingType=HalfHeightRadius
       width=0
       color {colorUseRule=UseColor_Specified, specifiedColor=<rgba 0 0 0 0>}
       cornerRadii=<0 0 0 0>
      }
      tintColor {colorUseRule=UseColor_Specified, specifiedColor=<rgba 1 1 1 1>}
      uVRegion {bIsValid=false, max=<0 0>, min=<0 0>}
     }
    }
   }
   slot {class=VerticalBoxSlot, properties {horizontalAlignment=HAlign_Left}}
  }
  Spacer_46 {widgetClassPath=&Spacer, parent=VerticalBox_6, widget=Spacer_46, properties {size=<1 16>}, slot {class=VerticalBoxSlot}}
  CabinEmailManagementWidget {
   widgetClassPath=/Game/UI/Foundation/CabinMode/ParentalControlsCabinEmailManagementWidget
   bInherited=true
   bIsVariable=true
   parent=VerticalBox_6
   widget=CabinEmailManagementWidget
   slot {class=VerticalBoxSlot}
  }
 ]
}
