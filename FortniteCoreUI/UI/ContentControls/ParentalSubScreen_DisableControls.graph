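# Parental Controls sub-screen that asks the player to confirm disabling parental controls.
# A centered vertical stack holds the rich-text question, side-by-side Yes and No buttons bound to the Confirm and Back input actions, and a "Disabling..." status line.
# The buttons and status text are inherited widgets of native parent FortParentalControls_DisableParentalControls.
# The event graph has only empty event stubs, so this asset supplies layout and styling only.
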WidgetBlueprint /FortniteCoreUI/UI/ContentControls/ParentalSubScreen_DisableControls
graphs [EventGraph {nodes [
   Event PreConstruct()
   Event Construct()
   Event Tick()
  ]}]
variables []
widgets {
 info {
  inheritedWidgetCount=3
  parentClass=/Script/FortniteUI.FortParentalControls_DisableParentalControls
  rootWidgetClass=&Overlay
  widgetCount=11
 }
 widgets [
  SubScreen_DisableControls {widgetClassPath=&Overlay, bIsVariable=true, widget=SubScreen_DisableControls}
  VerticalBox_6 {
   widgetClassPath=&VerticalBox
   parent=SubScreen_DisableControls
   widget=VerticalBox_6
   slot {class=OverlaySlot, properties {horizontalAlignment=HAlign_Fill, verticalAlignment=VAlign_Center}}
  }
  EnableDescription_1 {
   widgetClassPath=&CommonRichTextBlock
   parent=VerticalBox_6
   widget=EnableDescription_1
   properties {
    autoWrapText=true
    bOverrideDefaultStyle=true
    justification=Center
    text="Are you sure you want to disable <text color=\"00FFD6FF\" fontface=\"black\" case=\"upper\" size=\"30\">Parental Controls</>?"
    defaultTextStyleOverride {
     overflowPolicy=Clip
     transformPolicy=None
     colorAndOpacity {colorUseRule=UseColor_Specified, specifiedColor=<rgba 1 1 1 1>}
     font {
      bForceMonospaced=false
      bMaterialIsStencil=false
      fontMaterial=null
      fontObject=&BurbankSmall
      letterSpacing=0
      monospacedWidth=1
      size=28
      skewAmount=0
      typefaceFontName=Bold
      outlineSettings {
       bApplyOutlineToDropShadows=false
       bMiteredCorners=false
       bSeparateFillAlpha=false
       outlineMaterial=null
       outlineSize=0
       outlineColor=<rgba 0 0 0 1>
      }
     }
     highlightColor {colorUseRule=UseColor_Specified, specifiedColor=<rgba 0 0 0 1>}
     highlightShape {
      bIsDynamicallyLoaded=false
      drawAs=Image
      imageType=NoImage
      mirroring=NoMirror
      resourceName=None
      resourceObject=null
      tiling=NoTile
      imageSize=<32 32>
      margin=<ltrb 0 0 0 0>
      outlineSettings {
       bUseBrushTransparency=false
       roundingType=HalfHeightRadius
       width=0
       color {colorUseRule=UseColor_Specified, specifiedColor=<rgba 0 0 0 0>}
       cornerRadii=<0 0 0 0>
      }
      tintColor {colorUseRule=UseColor_Specified, specifiedColor=<rgba 1 1 1 1>}
      uVRegion {bIsValid=false, max=<0 0>, min=<0 0>}
     }
     selectedBackgroundColor {colorUseRule=UseColor_Specified, specifiedColor=<rgba 1 1 1 1>}
     shadowColorAndOpacity=<rgba 0 0 0 1>
     shadowOffset=<0 0>
     strikeBrush {
      bIsDynamicallyLoaded=false
      drawAs=Image
      imageType=NoImage
      mirroring=NoMirror
      resourceName=None
      resourceObject=null
      tiling=NoTile
      imageSize=<32 32>
      margin=<ltrb 0 0 0 0>
      outlineSettings {
       bUseBrushTransparency=false
       roundingType=HalfHeightRadius
       width=0
       color {colorUseRule=UseColor_Specified, specifiedColor=<rgba 0 0 0 0>}
       cornerRadii=<0 0 0 0>
      }
      tintColor {colorUseRule=UseColor_Specified, specifiedColor=<rgba 1 1 1 1>}
      uVRegion {bIsValid=false, max=<0 0>, min=<0 0>}
     }
     underlineBrush {
      bIsDynamicallyLoaded=false
      drawAs=Image
      imageType=NoImage
      mirroring=NoMirror
      resourceName=None
      resourceObject=null
      tiling=NoTile
      imageSize=<32 32>
      margin=<ltrb 0 0 0 0>
      outlineSettings {
       bUseBrushTransparency=false
       roundingType=HalfHeightRadius
       width=0
       color {colorUseRule=UseColor_Specified, specifiedColor=<rgba 0 0 0 0>}
       cornerRadii=<0 0 0 0>
      }
      tintColor {colorUseRule=UseColor_Specified, specifiedColor=<rgba 1 1 1 1>}
      uVRegion {bIsValid=false, max=<0 0>, min=<0 0>}
     }
    }
   }
   slot {class=VerticalBoxSlot}
  }
  Spacer_42 {widgetClassPath=&Spacer, parent=VerticalBox_6, widget=Spacer_42, properties {size=<1 48>}, slot {class=VerticalBoxSlot}}
  HorizontalBox_0 {
   widgetClassPath=&HorizontalBox
   parent=VerticalBox_6
   widget=HorizontalBox_0
   slot {class=VerticalBoxSlot, properties {size {sizeRule=Fill, value=1}}}
  }
  SizeBox_1 {
   widgetClassPath=&SizeBox
   parent=HorizontalBox_0
   widget=SizeBox_1
   properties {bOverride_WidthOverride=true, widthOverride=800}
   slot {class=HorizontalBoxSlot, properties {horizontalAlignment=HAlign_Center, size {sizeRule=Fill, value=1}}}
  }
  Button_DisableParentalControlsConfirm {
   widgetClassPath=&IconTextButton
   bInherited=true
   parent=SizeBox_1
   widget=Button_DisableParentalControlsConfirm
   properties {
    buttonText=Yes
    onButtonBaseClicked=(&self:WidgetTree.Button_DisableParentalControlsConfirm.HandleOnButtonClicked)
    onButtonBaseDoubleClicked=(&self:WidgetTree.Button_DisableParentalControlsConfirm.HandleOnButtonDoubleClicked)
    onButtonBaseHovered=(&self:WidgetTree.Button_DisableParentalControlsConfirm.HandleOnButtonHovered)
    onButtonBaseUnhovered=(&self:WidgetTree.Button_DisableParentalControlsConfirm.HandleOnButtonUnhovered)
    onSelectedChangedBase=(&self:WidgetTree.Button_DisableParentalControlsConfirm.HandleOnSelectedChanged)
    style=/FortUILibrary/Foundation/Buttons/ButtonStyles/Skew/ButtonStyle-Skew_Desirable.ButtonStyle-Skew_Desirable_C
    triggeringInputAction {dataTable=&InputActionDataTable, rowName=Confirm}
   }
   slot {class=SizeBoxSlot}
  }
  Spacer_55 {widgetClassPath=&Spacer, parent=HorizontalBox_0, widget=Spacer_55, properties {size=<50 1>}, slot {class=HorizontalBoxSlot}}
  SizeBox_0 {
   widgetClassPath=&SizeBox
   parent=HorizontalBox_0
   widget=SizeBox_0
   properties {bOverride_WidthOverride=true, widthOverride=800}
   slot {class=HorizontalBoxSlot, properties {horizontalAlignment=HAlign_Center, size {sizeRule=Fill, value=1}}}
  }
  Button_DisableParentalControlsCancel {
   widgetClassPath=&IconTextButton
   bInherited=true
   parent=SizeBox_0
   widget=Button_DisableParentalControlsCancel
   properties {
    buttonText=No
    onButtonBaseClicked=(&self:WidgetTree.Button_DisableParentalControlsCancel.HandleOnButtonClicked)
    onButtonBaseDoubleClicked=(&self:WidgetTree.Button_DisableParentalControlsCancel.HandleOnButtonDoubleClicked)
    onButtonBaseHovered=(&self:WidgetTree.Button_DisableParentalControlsCancel.HandleOnButtonHovered)
    onButtonBaseUnhovered=(&self:WidgetTree.Button_DisableParentalControlsCancel.HandleOnButtonUnhovered)
    onSelectedChangedBase=(&self:WidgetTree.Button_DisableParentalControlsCancel.HandleOnSelectedChanged)
    style=&ButtonStyle-Skew_Desirable_C
    triggeringInputAction {dataTable=&InputActionDataTable, rowName=Back}
   }
   slot {class=SizeBoxSlot}
  }
  Text_DisablingParentalControlsStatus {
   widgetClassPath=&CommonTextBlock
   bInherited=true
   parent=VerticalBox_6
   widget=Text_DisablingParentalControlsStatus
   properties {
    autoWrapText=true
    justification=Center
    text=Disabling...
    textTransformPolicy=ToUpper
    font {
     bForceMonospaced=false
     bMaterialIsStencil=false
     fontMaterial=null
     fontObject=&Burbank
     letterSpacing=0
     monospacedWidth=1
     size=36
     skewAmount=0
     typefaceFontName=Burbank-Black
     outlineSettings {
      bApplyOutlineToDropShadows=false
      bMiteredCorners=false
      bSeparateFillAlpha=false
      outlineMaterial=null
      outlineSize=0
      outlineColor=<rgba 0 0 0 1>
     }
    }
   }
   slot {class=VerticalBoxSlot, properties {horizontalAlignment=HAlign_Center, padding=<ltrb 0 24 0 0>}}
  }
 ]
}
