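# Content Controls settings panel widget, derived from FortSettingsPanel, that adds index tracking to the inherited settings list.
# Hover changes store the item's list index in CurrentSelectedIndex on hover and LastSelectedIndex on unhover; neither is read in this graph.
# When a list entry initializes as a SettingsListEntry_TimeWindows, the panel passes its list index into SettingIndex.
# The list takes its visual data from GameRegistryVisuals.
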WidgetBlueprint /FortniteCoreUI/UI/ContentControls/SettingsPanel_ContentControls
graphs [EventGraph {nodes [
   n1 = "On Item Is Hovered Changed (ListView_Settings)"!() -> (OutputDelegate, then, Item:Object Reference, bIsHovered:bool) {is=title K2Node_ComponentBoundEvent}
   > n2 = Branch(Condition=@n1.bIsHovered)
   n3 = GetIndexForItem!(
    self:List View Object Reference=Get ListView_Settings() -> (ListView_Settings:Fort Settings List View Object Reference)
    Item:Object Reference=@n1.Item
   ) -> (ReturnValue:int=0)
   > Set CurrentSelectedIndex(CurrentSelectedIndex:int=@n3) -> (Output_Get:int=0)
   Set LastSelectedIndex(execute=@n2.else, LastSelectedIndex:int=@n3) -> (Output_Get:int=0)
   n4 = "On Entry Initialized (ListView_Settings)"!() -> (OutputDelegate, then, Item:Object Reference, Widget:User Widget Object Reference) {is=title K2Node_ComponentBoundEvent}
   > n5 = Cast To SettingsListEntry_TimeWindows!(execute, Object=@n4.Widget) -> (then, CastFailed, AsSettings List Entry Time Windows:Settings List Entry Time Windows Object Reference) {is=title K2Node_DynamicCast}
   > Set SettingIndex!(
    execute
    SettingIndex:int=GetIndexForItem(self=Get ListView_Settings(), Item=@n4.Item) -> (ReturnValue=0)
    self:Settings List Entry Time Windows Object Reference=@n5
   ) -> (then, Output_Get:int=0)
   Event PreConstruct()
   Event Construct()
   Event Tick()
  ]}]
variables [
 CurrentSelectedIndex {type=integer, default=-1}
 LastSelectedIndex {type=integer, default=-1}
]
widgets {
 info {
  inheritedWidgetCount=2
  parentClass=/Script/FortniteUI.FortSettingsPanel
  rootWidgetClass=/Script/FortniteUI.FortSettingsListView
  widgetCount=2
 }
 widgets [
  ListView_Settings {
   widgetClassPath=&FortSettingsListView
   bInherited=true
   bIsVariable=true
   widget=ListView_Settings
   properties {
    bReturnFocusToSelection=true
    entryWidgetClass=/FortniteCoreUI/Settings/SettingsListEntry_Missing.SettingsListEntry_Missing_C
    horizontalEntrySpacing=4
    navigation=&self:WidgetTree.ListView_Settings.WidgetNavigation_0
    numDesignerPreviewEntries=8
    verticalEntrySpacing=4
    visualData=/FortniteCoreUI/Settings/GameRegistryVisuals
   }
  }
  Details_Settings {widgetClassPath=/Script/FortniteUI.FortSettingDetailView, bInherited=true, widget=null}
 ]
}
