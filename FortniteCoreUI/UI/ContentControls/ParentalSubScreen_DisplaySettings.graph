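# Parental controls display-settings sub-screen widget, built on FortParentalControls_DisplaySettings, with a settings panel, a More Settings info section, save-failure text, and Save and Disable buttons.
# On Construct it collapses the More Settings heading on mobile, then calls SetVisual.
# SetVisual reads the bEnableThyme runtime option and switches the heading's font, shear and color, plus the additional-settings and panel padding, between old and new looks.
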WidgetBlueprint /FortniteCoreUI/UI/ContentControls/ParentalSubScreen_DisplaySettings
graphs [
 EventGraph {nodes [
   Event Construct()
   > SetVisibility(
    self=Get CommonTextBlock_5() -> (CommonTextBlock_5:Common Text Object Reference)
    InVisibility=Select(Option 0=SelfHitTestInvisible, Option 1=Collapsed, Index=IsMobileGame())
   )
   > SetVisual()
   Event PreConstruct()
   Event Tick()
  ]}
 SetVisual {nodes [
   SetVisual() {is=title K2Node_FunctionEntry}
   > Set IsEnableThyme(IsEnableThyme:bool=Get bEnableThyme!(self:Fort Runtime Options Object Reference=GetRuntimeOptions()) -> (bEnableThyme:bool)) -> (Output_Get:bool)
   > SetFont(
    self=Get CommonTextBlock_5()
    InFontInfo=Select(
     Option 0:Slate Font Info Structure=Get MoreSettingsFont_Old() -> (MoreSettingsFont_Old:Slate Font Info Structure)
     Option 1:Slate Font Info Structure=Get MoreSettingsFont_New() -> (MoreSettingsFont_New:Slate Font Info Structure)
     Index=Get IsEnableThyme() -> (IsEnableThyme:bool)
    ) -> (ReturnValue:Slate Font Info Structure)
   )
   > SetRenderShear!(
    execute
    self:Widget Object Reference=Get CommonTextBlock_5()
    Shear:Vector 2D Structure=Select(Option 0:Vector 2D Structure=(X=-4,Y=0), Option 1:Vector 2D Structure, Index=Get IsEnableThyme()) -> (ReturnValue:Vector 2D Structure)
   ) -> (then)
   > SetColorAndOpacity(
    self=Get CommonTextBlock_5()
    InColorAndOpacity=Select(
     Option 0:Slate Color Structure=Get MoreSettingsColor_Old() -> (MoreSettingsColor_Old:Slate Color Structure)
     Option 1:Slate Color Structure=Get MoreSettingsColor_New() -> (MoreSettingsColor_New:Slate Color Structure)
     Index=Get IsEnableThyme()
    ) -> (ReturnValue:Slate Color Structure)
   )
   > SetPadding!(
    execute
    self:Vertical Box Slot Object Reference=SlotAsVerticalBoxSlot(Widget=Get AdditionalSettings() -> (AdditionalSettings:Vertical Box Object Reference))
    InPadding:Margin Structure=Select(
     Option 0:Margin Structure=Get AdditionalSettingsPadding_Old() -> (AdditionalSettingsPadding_Old:Margin Structure)
     Option 1:Margin Structure=Get AdditionalSettingsPadding_New() -> (AdditionalSettingsPadding_New:Margin Structure)
     Index=Get IsEnableThyme()
    ) -> (ReturnValue:Margin Structure)
   ) -> (then)
   n1 = Get IsEnableThyme()
   > SetPadding(
    self=SlotAsVerticalBoxSlot(Widget=Get SizeBox_0() -> (SizeBox_0:Size Box Object Reference))
    InPadding=Make Margin(Left=SelectFloat(A=-10, B=0, bPickA=@n1), Top=0, Right=0, Bottom=0)
   )
  ]}
]
variables []
widgets {
 info {
  inheritedWidgetCount=5
  parentClass=/Script/FortniteUI.FortParentalControls_DisplaySettings
  rootWidgetClass=&Overlay
  widgetCount=15
 }
 widgets [
  SubScreen_DisplaySettings {widgetClassPath=&Overlay, bIsVariable=true, widget=SubScreen_DisplaySettings}
  VerticalBox_95 {
   widgetClassPath=&VerticalBox
   parent=SubScreen_DisplaySettings
   widget=VerticalBox_95
   slot {class=OverlaySlot, properties {horizontalAlignment=HAlign_Fill, verticalAlignment=VAlign_Fill}}
  }
  SizeBox_0 {
   widgetClassPath=&SizeBox
   bIsVariable=true
   parent=VerticalBox_95
   widget=SizeBox_0
   properties {bOverride_MaxDesiredHeight=true, maxDesiredHeight=430}
   slot {class=VerticalBoxSlot, properties {padding=<ltrb -10 0 0 0>, size {sizeRule=Fill, value=1}}}
  }
  Panel_Settings {
   widgetClassPath=&/SettingsPanel_ContentControls
   bInherited=true
   parent=SizeBox_0
   widget=Panel_Settings
   slot {class=SizeBoxSlot}
  }
  AdditionalSettings {
   widgetClassPath=&VerticalBox
   bIsVariable=true
   parent=VerticalBox_95
   widget=AdditionalSettings
   slot {class=VerticalBoxSlot, properties {padding=<ltrb 0 20 0 20>}}
  }
  CommonTextBlock_5 {
   widgetClassPath=&CommonTextBlock
   bIsVariable=true
   parent=AdditionalSettings
   widget=CommonTextBlock_5
   properties {
    text=More Settings
    textTransformPolicy=ToUpper
    colorAndOpacity {colorUseRule=UseColor_Specified, specifiedColor=<rgba 0 1 0.6724 1>}
    font {
     bForceMonospaced=false
     bMaterialIsStencil=false
     fontMaterial=null
     fontObject=&Burbank
     letterSpacing=0
     monospacedWidth=1
     size=28
     skewAmount=0
     typefaceFontName=Burbank-Black
     outlineSettings {
      bApplyOutlineToDropShadows=false
      bMiteredCorners=false
      bSeparateFillAlpha=false
      outlineMaterial=null
      outlineSize=0
      outlineColor=<rgba 0 0 0 1>
     }
    }
    renderTransform {angle=0, scale=<1 1>, shear=<-4 0>, translation=<0 0>}
    shadowOffset=<0 0>
   }
   slot {class=VerticalBoxSlot}
  }
  SizeBox_7 {
   widgetClassPath=&SizeBox
   parent=AdditionalSettings
   widget=SizeBox_7
   properties {maxDesiredHeight=100}
   slot {class=VerticalBoxSlot, properties {padding=<ltrb 0 10 0 0>, size {sizeRule=Fill, value=1}}}
  }
  CommonTextBlock_12 {
   widgetClassPath=&CommonTextBlock
   parent=SizeBox_7
   widget=CommonTextBlock_12
   properties {
    autoWrapText=true
    lineHeightPercentage=1.29
    mobileFontSizeMultiplier=0.9
    style=/FortUILibrary/Styles/Typography/_TextStyle_Body1._TextStyle_Body1_C
    text="Use the button below for additional information about the Parental Controls and Settings on your platform, including restrictions on purchasing, playtime, communication, content viewing and sharing, and other social features."
    font {
     bForceMonospaced=false
     bMaterialIsStencil=false
     fontMaterial=null
     fontObject=&BurbankSmall
     letterSpacing=0
     monospacedWidth=1
     size=18
     skewAmount=0
     typefaceFontName=Bold
     outlineSettings {
      bApplyOutlineToDropShadows=false
      bMiteredCorners=false
      bSeparateFillAlpha=false
      outlineMaterial=null
      outlineSize=0
      outlineColor=<rgba 0 0 0 1>
     }
    }
    shadowOffset=<0 0>
    strikeBrush {
     bIsDynamicallyLoaded=false
     drawAs=NoDrawType
     imageType=NoImage
     mirroring=NoMirror
     resourceName=None
     resourceObject=null
     tiling=NoTile
     imageSize=<32 32>
     margin=<ltrb 0 0 0 0>
     outlineSettings {
      bUseBrushTransparency=false
      roundingType=HalfHeightRadius
      width=0
      color {colorUseRule=UseColor_Specified, specifiedColor=<rgba 0 0 0 0>}
      cornerRadii=<0 0 0 0>
     }
     tintColor {colorUseRule=UseColor_Specified, specifiedColor=<rgba 1 1 1 1>}
     uVRegion {bIsValid=false, max=<0 0>, min=<0 0>}
    }
   }
   slot {class=SizeBoxSlot}
  }
  Button_MoreSettings {
   widgetClassPath=&IconTextButton
   bInherited=true
   parent=AdditionalSettings
   widget=Button_MoreSettings
   properties {
    buttonText=More Settings
    onButtonBaseClicked=(&self:WidgetTree.Button_MoreSettings.HandleOnButtonClicked)
    onButtonBaseDoubleClicked=(&self:WidgetTree.Button_MoreSettings.HandleOnButtonDoubleClicked)
    onButtonBaseHovered=(&self:WidgetTree.Button_MoreSettings.HandleOnButtonHovered)
    onButtonBaseUnhovered=(&self:WidgetTree.Button_MoreSettings.HandleOnButtonUnhovered)
    onSelectedChangedBase=(&self:WidgetTree.Button_MoreSettings.HandleOnSelectedChanged)
    style=/FortUILibrary/Foundation/Buttons/ButtonStyles/Skew/ButtonStyle-Skew_Error.ButtonStyle-Skew_Error_C
    triggeringInputAction {dataTable=&InputActionDataTable, rowName=GenericFaceButtonTop}
   }
   slot {class=VerticalBoxSlot, properties {padding=<ltrb 50 10 50 0>}}
  }
  Text_SaveFail {
   widgetClassPath=&CommonTextBlock
   bInherited=true
   parent=VerticalBox_95
   widget=Text_SaveFail
   properties {
    autoWrapText=true
    justification=Center
    text=Failed To Save
    font {
     bForceMonospaced=false
     bMaterialIsStencil=false
     fontMaterial=null
     fontObject=&NotoSans
     letterSpacing=0
     monospacedWidth=1
     size=20
     skewAmount=0
     typefaceFontName=NotoSans-Bold
     outlineSettings {
      bApplyOutlineToDropShadows=false
      bMiteredCorners=false
      bSeparateFillAlpha=false
      outlineMaterial=null
      outlineSize=0
      outlineColor=<rgba 0 0 0 1>
     }
    }
    shadowOffset=<0 0>
   }
   slot {class=VerticalBoxSlot, properties {padding=<ltrb 0 10 0 0>}}
  }
  HorizontalBox_1 {
   widgetClassPath=&HorizontalBox
   parent=VerticalBox_95
   widget=HorizontalBox_1
   slot {class=VerticalBoxSlot, properties {horizontalAlignment=HAlign_Center}}
  }
  SizeBox_1 {
   widgetClassPath=&SizeBox
   parent=HorizontalBox_1
   widget=SizeBox_1
   properties {bOverride_MinDesiredWidth=true, minDesiredWidth=400}
   slot {class=HorizontalBoxSlot}
  }
  Button_SaveSettings {
   widgetClassPath=&IconTextButton
   bInherited=true
   parent=SizeBox_1
   widget=Button_SaveSettings
   properties {
    buttonText=Save
    onButtonBaseClicked=(&self:WidgetTree.Button_SaveSettings.HandleOnButtonClicked)
    onButtonBaseDoubleClicked=(&self:WidgetTree.Button_SaveSettings.HandleOnButtonDoubleClicked)
    onButtonBaseHovered=(&self:WidgetTree.Button_SaveSettings.HandleOnButtonHovered)
    onButtonBaseUnhovered=(&self:WidgetTree.Button_SaveSettings.HandleOnButtonUnhovered)
    onSelectedChangedBase=(&self:WidgetTree.Button_SaveSettings.HandleOnSelectedChanged)
    style=/FortUILibrary/Foundation/Buttons/ButtonStyles/Skew/ButtonStyle-Skew_Desirable.ButtonStyle-Skew_Desirable_C
    triggeringInputAction {dataTable=&InputActionDataTable, rowName=GenericFaceButtonLeft}
   }
   slot {class=SizeBoxSlot}
  }
  SizeBox_2 {
   widgetClassPath=&SizeBox
   parent=HorizontalBox_1
   widget=SizeBox_2
   properties {bOverride_MinDesiredWidth=true, minDesiredWidth=400}
   slot {class=HorizontalBoxSlot, properties {padding=<ltrb 100 0 0 0>, size {sizeRule=Fill, value=1}}}
  }
  Button_DisableParentalControls {
   widgetClassPath=&IconTextButton
   bInherited=true
   parent=SizeBox_2
   widget=Button_DisableParentalControls
   properties {
    buttonText=Disable
    onButtonBaseClicked=(&self:WidgetTree.Button_DisableParentalControls.HandleOnButtonClicked)
    onButtonBaseDoubleClicked=(&self:WidgetTree.Button_DisableParentalControls.HandleOnButtonDoubleClicked)
    onButtonBaseHovered=(&self:WidgetTree.Button_DisableParentalControls.HandleOnButtonHovered)
    onButtonBaseUnhovered=(&self:WidgetTree.Button_DisableParentalControls.HandleOnButtonUnhovered)
    onSelectedChangedBase=(&self:WidgetTree.Button_DisableParentalControls.HandleOnSelectedChanged)
    style=&ButtonStyle-Skew_Desirable_C
    triggeringInputAction {dataTable=&InputActionDataTable, rowName=GenericRightShoulder}
   }
   slot {class=SizeBoxSlot}
  }
 ]
}
