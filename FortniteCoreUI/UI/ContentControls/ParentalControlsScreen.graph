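# Parental Controls screen, a FortParentalControls widget hosting the PIN, email, enable/disable and settings sub-views.
# Sub-screen changes swap the right panel between side art and, for DisplaySettings, the settings detail view; exit-action changes relabel the close button Back/Exit.
# Unless a deprecated focus-hack runtime option is set, a migration-modal router event parks the sub-view switcher on an empty spacer, then restores it.
# PreConstruct tightens padding for compact UI.
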WidgetBlueprint /FortniteCoreUI/UI/ContentControls/ParentalControlsScreen
graphs [
 EventGraph {nodes [
   n1 = Event OnNewSubScreenSet!() -> (OutputDelegate, then, ScreenState:EParentalControlsViewState Enum=Invalid) {is=title K2Node_Event}
   n2 = Get Side_Image() -> (Side_Image:Image Object Reference)
   > SetActiveWidget(
    self:Widget Switcher Object Reference=Get RightSideSwitcher() -> (RightSideSwitcher:Common Widget Switcher Legacy Object Reference)
    Widget=Select!(
     Invalid:Widget Object Reference=@n2
     EnterPin:Widget Object Reference=@n2
     AskToEnableControls:Widget Object Reference=@n2
     VerifyEmail:Widget Object Reference=@n2
     SetupEmail:Widget Object Reference=@n2
     SetupPin:Widget Object Reference=@n2
     DisplaySettings:Widget Object Reference=Get SettingsContainer() -> (SettingsContainer:Size Box Object Reference)
     DisableParentalControls:Widget Object Reference=@n2
     AskToReEnable:Widget Object Reference=@n2
     ReEnabling:Widget Object Reference=@n2
     CabinModeEmailChange:Widget Object Reference=@n2
     Index:EParentalControlsViewState Enum=@n1.ScreenState
    ) -> (ReturnValue:Widget Object Reference)
   )
   n3 = Event OnExitActionChange!() -> (OutputDelegate, then, bIsExitAction:bool) {is=title K2Node_Event}
   > n4 = Cast To IconTextButton!(execute, Object=Get Button_CloseScreen() -> (Button_CloseScreen:Common Button Legacy Object Reference)) -> (then, CastFailed, AsIcon Text Button:Icon Text Button Object Reference) {is=title K2Node_DynamicCast}
   > Set Text!(
    execute
    self:Icon Text Button Object Reference=@n4
    ButtonText:Text=Select!(Option 0:Text="LOCTEXT(\"Back\")", Option 1:Text="LOCTEXT(\"Exit\")", Index:bool=@n3.bIsExitAction) -> (ReturnValue:Text)
   ) -> (then) {is=title K2Node_CallFunction}
   Event Construct()
   > n5 = Branch(Condition=Get bDeprecatedSettingModalUseCodeDrivenFocusHack!(self:Fort Runtime Options Object Reference=GetRuntimeOptions()) -> (bDeprecatedSettingModalUseCodeDrivenFocusHack:bool))
   n6 = StartListeningToEvent!(
    execute:Exec=@n5.else
    Target:Gameplay Event Router Component Object Reference=GetGlobalEventRouter()
    EventType:Script Struct Object Reference=/FortniteCoreUI/UI/ContentControls/Dialog/BPS_ParentalControlsMigrationSettingsModalActiveChange
    Context:Object Reference=GetOwningPlayer()
    EventBubblingRule:EEventBubblingRule Enum
   ) -> (
    then:Exec
    AsyncTaskProxy:Async Action Start Listening To Event Object Reference
    OnEventReceived:Exec
    Payload:BPS Parental Controls Migration Settings Modal Active Change Structure
   )
   > Set AsyncModalEventReference(AsyncModalEventReference:Async Action Start Listening To Event Object Reference=@n6) -> (Output_Get:Async Action Start Listening To Event Object Reference)
   n7 = Branch(
    execute=@n6.OnEventReceived
    Condition=Break BPS Parental Controls Migration Settings Modal Active Change!(BPS_ParentalControlsMigrationSettingsModalActiveChange:"BPS Parental Controls Migration Settings Modal Active Change Structure (by ref)"=@n6.Payload) -> (bActive_4_#1:bool=~) {is=title K2Node_BreakStruct}
   )
   > Set ActiveWidgetBeforeModal(ActiveWidgetBeforeModal:Widget Object Reference=GetActiveWidget!(self:Widget Switcher Object Reference=Get Switcher_SubViews() -> (Switcher_SubViews:Common Activatable Widget Switcher Object Reference)) -> (ReturnValue:Widget Object Reference)) -> (Output_Get:Widget Object Reference)
   > SetActiveWidget(self=Get Switcher_SubViews(), Widget=Get SubView_Empty() -> (SubView_Empty:Spacer Object Reference))
   SetActiveWidget(execute=@n7.else, self=Get Switcher_SubViews(), Widget=Get ActiveWidgetBeforeModal() -> (ActiveWidgetBeforeModal:Widget Object Reference))
   Event Destruct()
   > n8 = Branch(Condition=Get bDeprecatedSettingModalUseCodeDrivenFocusHack(self=GetRuntimeOptions()))
   StopListeningToEvent(
    execute=@n8.else
    self:Async Action Start Listening To Event Object Reference=Get AsyncModalEventReference() -> (AsyncModalEventReference:Async Action Start Listening To Event Object Reference)
   )
   Event PreConstruct()
   > SetCompactUIStyle() {is=title K2Node_CallFunction}
  ]}
 SetCompactUIStyle {nodes [
   SetCompactUIStyle() {is=title K2Node_FunctionEntry}
   > Branch(Condition=IsUsingCompactUI())
   > SetPadding!(
    execute
    self:Vertical Box Slot Object Reference=SlotAsVerticalBoxSlot(Widget=Get VB_Content() -> (VB_Content:Vertical Box Object Reference))
    InPadding_Left:float=48
    InPadding_Top:float=24
    InPadding_Right:float=48
    InPadding_Bottom:float=24
   ) -> (then)
   > SetPadding(
    self:Border Slot Object Reference=SlotAsBorderSlot!(Widget:Widget Object Reference=Get VisibilityBorder_NonTouch() -> (VisibilityBorder_NonTouch:Common Hardware Visibility Border Object Reference)) -> (ReturnValue:Border Slot Object Reference)
    InPadding_Left=0
    InPadding_Top=0
    InPadding_Right=48
    InPadding_Bottom=8
   )
   > SetPadding(
    self:Overlay Slot Object Reference=SlotAsOverlaySlot(Widget=Get ContentRatingTooltip() -> (ContentRatingTooltip:Fort Sidebar Onboard Tooltip Widget Object Reference))
    InPadding_Left=0
    InPadding_Top=0
    InPadding_Right=0
    InPadding_Bottom=0
   )
   > SetPadding(
    self=SlotAsOverlaySlot(Widget=Get UnlockedContentTooltip() -> (UnlockedContentTooltip:Fort Sidebar Onboard Tooltip Widget Object Reference))
    InPadding_Left=0
    InPadding_Top=-64
    InPadding_Right=0
    InPadding_Bottom=0
   )
   > SetPadding(
    self=SlotAsOverlaySlot(Widget=Get CreativeUnlockTooltip() -> (CreativeUnlockTooltip:Fort Sidebar Onboard Tooltip Widget Object Reference))
    InPadding_Left=0
    InPadding_Top=25
    InPadding_Right=0
    InPadding_Bottom=0
   )
   > SetPadding(
    self:Widget Switcher Slot Object Reference=SlotAsWidgetSwitcherSlot!(Widget:Widget Object Reference=Get SettingsContainer()) -> (ReturnValue:Widget Switcher Slot Object Reference)
    InPadding_Left=0
    InPadding_Top=0
    InPadding_Right=0
    InPadding_Bottom=132
   )
  ]}
]
variables [
 ActiveWidgetBeforeModal {type=Widget, default=null}
 AsyncModalEventReference {type=AsyncAction_StartListeningToEvent, default=null}
]
widgets {
 info {inheritedWidgetCount=19, parentClass=/Script/FortniteUI.FortParentalControls, rootWidgetClass=&Overlay, widgetCount=42}
 widgets [
  Overlay_0 {widgetClassPath=&Overlay, widget=Overlay_0, properties {clipping=ClipToBounds, visibility=Visible}}
  BackgroundGradient {
   widgetClassPath=&Image
   parent=Overlay_0
   widget=BackgroundGradient
   properties {
    visibility=SelfHitTestInvisible
    brush {
     bIsDynamicallyLoaded=false
     drawAs=Image
     imageType=NoImage
     mirroring=NoMirror
     resourceName=None
     resourceObject=/FortniteCoreUI/Frontend/Art/M_UI_RadialGradient_Right
     tiling=NoTile
     imageSize=<32 32>
     margin=<ltrb 0 0 0 0>
     outlineSettings {
      bUseBrushTransparency=false
      roundingType=HalfHeightRadius
      width=0
      color {colorUseRule=UseColor_Specified, specifiedColor=<rgba 0 0 0 0>}
      cornerRadii=<0 0 0 0>
     }
     tintColor {colorUseRule=UseColor_Specified, specifiedColor=<rgba 1 1 1 1>}
     uVRegion {bIsValid=false, max=<0 0>, min=<0 0>}
    }
   }
   slot {class=OverlaySlot, properties {horizontalAlignment=HAlign_Fill, verticalAlignment=VAlign_Fill}}
  }
  WBP_CaptureForPostBufferUpdate {
   widgetClassPath=&WBP_CaptureForPostBufferUpdate
   bIsVariable=true
   parent=Overlay_0
   slot {class=OverlaySlot}
  }
  SafeZone {
   widgetClassPath=&SafeZone
   bIsVariable=true
   parent=Overlay_0
   widget=SafeZone
   slot {class=OverlaySlot, properties {horizontalAlignment=HAlign_Fill, verticalAlignment=VAlign_Fill}}
  }
  Overlay_3 {widgetClassPath=&Overlay, parent=SafeZone, widget=Overlay_3, slot {class=SafeZoneSlot}}
  VerticalBox_0 {
   widgetClassPath=&VerticalBox
   parent=Overlay_3
   widget=VerticalBox_0
   slot {class=OverlaySlot, properties {horizontalAlignment=HAlign_Fill, verticalAlignment=VAlign_Fill}}
  }
  VB_Content {
   widgetClassPath=&VerticalBox
   bIsVariable=true
   parent=VerticalBox_0
   widget=VB_Content
   slot {class=VerticalBoxSlot, properties {padding=<ltrb 60 64 60 24>, size {sizeRule=Fill, value=1}}}
  }
  Border_0 {
   widgetClassPath=&Border
   parent=VB_Content
   widget=Border_0
   properties {
    background {
     bIsDynamicallyLoaded=false
     drawAs=NoDrawType
     imageType=NoImage
     mirroring=NoMirror
     resourceName=None
     resourceObject=null
     tiling=NoTile
     imageSize=<32 32>
     margin=<ltrb 0 0 0 0>
     outlineSettings {
      bUseBrushTransparency=false
      roundingType=HalfHeightRadius
      width=0
      color {colorUseRule=UseColor_Specified, specifiedColor=<rgba 0 0 0 0>}
      cornerRadii=<0 0 0 0>
     }
     tintColor {colorUseRule=UseColor_Specified, specifiedColor=<rgba 1 1 1 1>}
     uVRegion {bIsValid=false, max=<0 0>, min=<0 0>}
    }
    contentColorAndOpacity=<rgba 0.6038 1 1 1>
    padding=<ltrb 0 0 0 0>
   }
   slot {class=VerticalBoxSlot}
  }
  ScreenTitle {
   widgetClassPath=&CommonTextBlock
   parent=Border_0
   widget=ScreenTitle
   properties {
    style=/FortUILibrary/Styles/Typography/_TextStyle_SubH2._TextStyle_SubH2_C
    text=Parental Controls
    textTransformPolicy=ToUpper
    font {
     bForceMonospaced=false
     bMaterialIsStencil=false
     fontMaterial=null
     fontObject=&Burbank
     letterSpacing=0
     monospacedWidth=1
     size=33
     skewAmount=0
     typefaceFontName=BurbankRegular-Black
     outlineSettings {
      bApplyOutlineToDropShadows=false
      bMiteredCorners=false
      bSeparateFillAlpha=false
      outlineMaterial=null
      outlineSize=0
      outlineColor=<rgba 0 0 0 1>
     }
    }
    shadowOffset=<0 0>
    strikeBrush {
     bIsDynamicallyLoaded=false
     drawAs=NoDrawType
     imageType=NoImage
     mirroring=NoMirror
     resourceName=None
     resourceObject=null
     tiling=NoTile
     imageSize=<32 32>
     margin=<ltrb 0 0 0 0>
     outlineSettings {
      bUseBrushTransparency=false
      roundingType=HalfHeightRadius
      width=0
      color {colorUseRule=UseColor_Specified, specifiedColor=<rgba 0 0 0 0>}
      cornerRadii=<0 0 0 0>
     }
     tintColor {colorUseRule=UseColor_Specified, specifiedColor=<rgba 1 1 1 1>}
     uVRegion {bIsValid=false, max=<0 0>, min=<0 0>}
    }
   }
   slot {class=BorderSlot, properties {padding=<ltrb 0 0 0 0>}}
  }
  Spacer {widgetClassPath=&Spacer, parent=VB_Content, widget=Spacer, properties {size=<1 24>}, slot {class=VerticalBoxSlot}}
  HorizontalBox_41 {
   widgetClassPath=&HorizontalBox
   parent=VB_Content
   widget=HorizontalBox_41
   slot {class=VerticalBoxSlot, properties {size {sizeRule=Fill, value=1}}}
  }
  VerticalBox_2 {
   widgetClassPath=&VerticalBox
   parent=HorizontalBox_41
   widget=VerticalBox_2
   slot {class=HorizontalBoxSlot, properties {size {sizeRule=Fill, value=0.5}}}
  }
  Switcher_SubViews {
   widgetClassPath=&CommonActivatableWidgetSwitcher
   bInherited=true
   bIsVariable=true
   parent=VerticalBox_2
   widget=Switcher_SubViews
   properties {activeWidgetIndex=10, transitionDuration=0}
   slot {class=VerticalBoxSlot, properties {size {sizeRule=Fill, value=1}}}
  }
  SubView_AskToEnableControls {
   widgetClassPath=&/ParentalSubScreen_AskToEnable
   bInherited=true
   bIsVariable=true
   parent=Switcher_SubViews
   widget=SubView_AskToEnableControls
   slot {class=WidgetSwitcherSlot}
  }
  SubView_AskToReEnable {
   widgetClassPath=&/ParentalSubScreen_AskToReEnable
   bInherited=true
   bIsVariable=true
   parent=Switcher_SubViews
   widget=SubView_AskToReEnable
   slot {class=WidgetSwitcherSlot}
  }
  SubView_DisableParentalControls {
   widgetClassPath=&/ParentalSubScreen_DisableControls
   bInherited=true
   bIsVariable=true
   parent=Switcher_SubViews
   widget=SubView_DisableParentalControls
   slot {class=WidgetSwitcherSlot}
  }
  SubView_DisplaySettings {
   widgetClassPath=&/ParentalSubScreen_DisplaySettings
   bInherited=true
   bIsVariable=true
   parent=Switcher_SubViews
   widget=SubView_DisplaySettings
   slot {class=WidgetSwitcherSlot}
  }
  SubView_Renabling {
   widgetClassPath=&/ParentalSubScreen_ReEnabling
   bInherited=true
   bIsVariable=true
   parent=Switcher_SubViews
   widget=SubView_Renabling
   slot {class=WidgetSwitcherSlot}
  }
  SubView_SetupEmail {
   widgetClassPath=&/ParentalSubScreen_SetupEmail
   bInherited=true
   bIsVariable=true
   parent=Switcher_SubViews
   widget=SubView_SetupEmail
   slot {class=WidgetSwitcherSlot}
  }
  SubView_SetupPin {
   widgetClassPath=&/ParentalSubScreen_SetupPin
   bInherited=true
   bIsVariable=true
   parent=Switcher_SubViews
   widget=SubView_SetupPin
   slot {class=WidgetSwitcherSlot}
  }
  SubView_VerifyEmail {
   widgetClassPath=&/ParentalSubScreen_VerifyEmail
   bInherited=true
   bIsVariable=true
   parent=Switcher_SubViews
   widget=SubView_VerifyEmail
   slot {class=WidgetSwitcherSlot}
  }
  SubView_CabinModeEmailChange {
   widgetClassPath=&/ParentalSubScreen_CabinMode
   bInherited=true
   bIsVariable=true
   parent=Switcher_SubViews
   widget=SubView_CabinModeEmailChange
   slot {class=WidgetSwitcherSlot}
  }
  SubView_EnterPin {
   widgetClassPath=&/ParentalSubScreen_EnterPin
   bInherited=true
   bIsVariable=true
   parent=Switcher_SubViews
   widget=SubView_EnterPin
   slot {class=WidgetSwitcherSlot}
  }
  SubView_Empty {widgetClassPath=&Spacer, bIsVariable=true, parent=Switcher_SubViews, widget=SubView_Empty, slot {class=WidgetSwitcherSlot}}
  Spacer_68 {widgetClassPath=&Spacer, parent=HorizontalBox_41, widget=Spacer_68, properties {size=<48 1>}, slot {class=HorizontalBoxSlot}}
  RightSideSwitcher {
   widgetClassPath=&CommonWidgetSwitcherLegacy
   bIsVariable=true
   parent=HorizontalBox_41
   widget=RightSideSwitcher
   slot {class=HorizontalBoxSlot, properties {size {sizeRule=Fill, value=0.5}}}
  }
  Side_Image {
   widgetClassPath=&Image
   bIsVariable=true
   parent=RightSideSwitcher
   widget=Side_Image
   properties {brush {
     bIsDynamicallyLoaded=false
     drawAs=Image
     imageType=NoImage
     mirroring=NoMirror
     resourceName=None
     resourceObject=/Game/UI/ContentControls/MaterialInstances/MI_UI_ParentalControls_Imagery
     tiling=NoTile
     imageSize=<32 32>
     margin=<ltrb 0 0 0 0>
     outlineSettings {
      bUseBrushTransparency=false
      roundingType=HalfHeightRadius
      width=0
      color {colorUseRule=UseColor_Specified, specifiedColor=<rgba 0 0 0 0>}
      cornerRadii=<0 0 0 0>
     }
     tintColor {colorUseRule=UseColor_Specified, specifiedColor=<rgba 1 1 1 1>}
     uVRegion {bIsValid=false, max=<0 0>, min=<0 0>}
    }}
   slot {class=WidgetSwitcherSlot, properties {padding=<ltrb 0 -84 0 -24>}}
  }
  SettingsContainer {
   widgetClassPath=&SizeBox
   bIsVariable=true
   parent=RightSideSwitcher
   widget=SettingsContainer
   properties {bOverride_WidthOverride=true, widthOverride=800}
   slot {class=WidgetSwitcherSlot}
  }
  Details_Settings {
   widgetClassPath=/FortniteCoreUI/Settings/ParentalControls_DetailView
   bInherited=true
   bIsVariable=true
   parent=SettingsContainer
   widget=Details_Settings
   slot {class=SizeBoxSlot, properties {verticalAlignment=VAlign_Top}}
  }
  Footer {
   widgetClassPath=&Border
   parent=VerticalBox_0
   widget=Footer
   properties {
    horizontalAlignment=HAlign_Right
    verticalAlignment=VAlign_Bottom
    background {
     bIsDynamicallyLoaded=false
     drawAs=Image
     imageType=NoImage
     mirroring=NoMirror
     resourceName=None
     resourceObject=null
     tiling=NoTile
     imageSize=<32 32>
     margin=<ltrb 0 0 0 0>
     outlineSettings {
      bUseBrushTransparency=false
      roundingType=HalfHeightRadius
      width=0
      color {colorUseRule=UseColor_Specified, specifiedColor=<rgba 0 0 0 0>}
      cornerRadii=<0 0 0 0>
     }
     tintColor {colorUseRule=UseColor_Specified, specifiedColor=<rgba 0.0024 0.016 0.117 1>}
     uVRegion {bIsValid=false, max=<0 0>, min=<0 0>}
    }
    padding=<ltrb 60 8 60 8>
   }
   slot {class=VerticalBoxSlot}
  }
  VisibilityBorder_NonTouch {
   widgetClassPath=&CommonHardwareVisibilityBorder
   bIsVariable=true
   parent=Footer
   widget=VisibilityBorder_NonTouch
   properties {visibilityQuery {
     autoDescription=" NONE( Input.Touch )"
     tokenStreamVersion=0
     userDescription=""
     queryTokenStream [0,1,3,1,0]
     tagDictionary [{tagName=Input.Touch}]
    }}
   slot {class=BorderSlot, properties {horizontalAlignment=HAlign_Right, verticalAlignment=VAlign_Bottom, padding=<ltrb 60 8 60 8>}}
  }
  Button_CloseScreen {
   widgetClassPath=&IconTextButton
   bInherited=true
   bIsVariable=true
   parent=VisibilityBorder_NonTouch
   widget=Button_CloseScreen
   properties {
    buttonText=Exit
    onButtonBaseClicked=(&self:WidgetTree.Button_CloseScreen.HandleOnButtonClicked)
    onButtonBaseDoubleClicked=(&self:WidgetTree.Button_CloseScreen.HandleOnButtonDoubleClicked)
    onButtonBaseHovered=(&self:WidgetTree.Button_CloseScreen.HandleOnButtonHovered)
    onButtonBaseUnhovered=(&self:WidgetTree.Button_CloseScreen.HandleOnButtonUnhovered)
    onSelectedChangedBase=(&self:WidgetTree.Button_CloseScreen.HandleOnSelectedChanged)
    style=/FortUILibrary/Foundation/Buttons/ButtonStyles/Skew/ButtonStyle-Skew_Error.ButtonStyle-Skew_Error_C
    triggeringInputAction {dataTable=&InputActionDataTable, rowName=Close}
   }
   slot {class=BorderSlot, properties {padding=<ltrb 0 0 0 0>}}
  }
  VisibilityBorder_Touch {
   widgetClassPath=&CommonHardwareVisibilityBorder
   parent=Overlay_3
   widget=VisibilityBorder_Touch
   properties {
    visibleType=Visible
    visibilityQuery {
     autoDescription=" ANY( Input.Touch )"
     tokenStreamVersion=0
     userDescription=""
     queryTokenStream [0,1,1,1,0]
     tagDictionary [{tagName=Input.Touch}]
    }
   }
   slot {class=OverlaySlot, properties {horizontalAlignment=HAlign_Right, padding=<ltrb 0 64 60 0>}}
  }
  Button_CloseTouch {
   widgetClassPath=/FortUILibrary/Foundation/Buttons/CloseButton
   bInherited=true
   bIsVariable=true
   parent=VisibilityBorder_Touch
   widget=Button_CloseTouch
   properties {
    onButtonBaseClicked=(&self:WidgetTree.Button_CloseTouch.HandleOnButtonClicked)
    onButtonBaseDoubleClicked=(&self:WidgetTree.Button_CloseTouch.HandleOnButtonDoubleClicked)
    onButtonBaseHovered=(&self:WidgetTree.Button_CloseTouch.HandleOnButtonHovered)
    onButtonBaseUnhovered=(&self:WidgetTree.Button_CloseTouch.HandleOnButtonUnhovered)
    onSelectedChangedBase=(&self:WidgetTree.Button_CloseTouch.HandleOnSelectedChanged)
   }
   slot {class=BorderSlot, properties {padding=<ltrb 0 0 0 0>}}
  }
  Overlay_Tooltips {
   widgetClassPath=&Overlay
   parent=Overlay_3
   widget=Overlay_Tooltips
   properties {renderTransform {angle=0, scale=<1 1>, shear=<0 0>, translation=<850 90>}}
   slot {class=OverlaySlot}
  }
  ContentRatingTooltip {
   widgetClassPath=&/ParentalControlsContentRatingsTooltip
   bInherited=true
   bIsVariable=true
   parent=Overlay_Tooltips
   widget=ContentRatingTooltip
   slot {class=OverlaySlot, properties {padding=<ltrb 0 40 0 0>}}
  }
  UnlockedContentTooltip {
   widgetClassPath=&/ParentalControlsUnlockedContentTooltip
   bInherited=true
   bIsVariable=true
   parent=Overlay_Tooltips
   widget=UnlockedContentTooltip
   slot {class=OverlaySlot}
  }
  CreativeUnlockTooltip {
   widgetClassPath=&/ParentalControlsCreativeTooltip
   bInherited=true
   bIsVariable=true
   parent=Overlay_Tooltips
   widget=CreativeUnlockTooltip
   properties {socialTutorial=CreativeContentSettings}
   slot {class=OverlaySlot, properties {padding=<ltrb 0 80 0 0>}}
  }
  Overlay_SavingSettings {
   widgetClassPath=&Overlay
   bInherited=true
   parent=Overlay_0
   widget=Overlay_SavingSettings
   slot {class=OverlaySlot, properties {horizontalAlignment=HAlign_Fill, verticalAlignment=VAlign_Fill}}
  }
  Image_67 {
   widgetClassPath=&Image
   bIsVariable=true
   parent=Overlay_SavingSettings
   widget=Image_67
   properties {brush {
     bIsDynamicallyLoaded=false
     drawAs=Image
     imageType=NoImage
     mirroring=NoMirror
     resourceName=None
     resourceObject=null
     tiling=NoTile
     imageSize=<32 32>
     margin=<ltrb 0 0 0 0>
     outlineSettings {
      bUseBrushTransparency=false
      roundingType=HalfHeightRadius
      width=0
      color {colorUseRule=UseColor_Specified, specifiedColor=<rgba 0 0 0 0>}
      cornerRadii=<0 0 0 0>
     }
     tintColor {colorUseRule=UseColor_Specified, specifiedColor=<rgba 0 0 0 0.5>}
     uVRegion {bIsValid=false, max=<0 0>, min=<0 0>}
    }}
   slot {class=OverlaySlot, properties {horizontalAlignment=HAlign_Fill, verticalAlignment=VAlign_Fill}}
  }
  CircularThrobber_72 {
   widgetClassPath=/Script/UMG.CircularThrobber
   bIsVariable=true
   parent=Overlay_SavingSettings
   widget=CircularThrobber_72
   slot {class=OverlaySlot, properties {horizontalAlignment=HAlign_Center, verticalAlignment=VAlign_Center}}
  }
  UpdateNotice {
   widgetClassPath=/FortniteCoreUI/UI/Foundation/Modals/Settings/WBP_Modal_ParentalControlsUpdated
   bInherited=true
   bIsVariable=true
   parent=Overlay_0
   widget=UpdateNotice
   properties {visibility=Collapsed}
   slot {class=OverlaySlot, properties {horizontalAlignment=HAlign_Fill, verticalAlignment=VAlign_Fill}}
  }
 ]
}
