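# Activatable CommonUI panel that blocks content purchase pending account setup, showing a title, detail text and an Open Parental Controls button.
# PreConstruct pushes the detail text and applies Max Width as the wrap width for both text blocks.
# Clicking the button, which is also the desired focus target, calls RequestToOpenParentalControls.
# An input-method listener, bound from Construct to Destruct, applies a fixed button height only for touch.
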WidgetBlueprint /FortniteCoreUI/UI/ContentControls/WBP_ContentPurchaseBlocker
graphs [
 BP_GetDesiredFocusTarget {nodes [
   Get Desired Focus Target()
   > Return Node!(execute, ReturnValue:Widget Object Reference=Get WBP_UIKit_Button_Loud() -> (WBP_UIKit_Button_Loud:WBP UIKit Button Loud Object Reference))
  ]}
 EventGraph {nodes [
   "On Clicked (WBP_UIKit_Button_Loud)"!() -> (OutputDelegate, then, Button:Common Button Base Object Reference) {is=title K2Node_ComponentBoundEvent}
   > RequestToOpenParentalControls!(execute) -> (then)
   n1 = OnInputMethodChanged!() -> (OutputDelegate, then, bNewInputType:ECommonInputType Enum) {is=title K2Node_CustomEvent}
   > n2 = Branch(Condition=Select!(MouseAndKeyboard:bool=false, Gamepad:bool=false, Touch:bool=true, Count:bool=false, Index:ECommonInputType Enum=@n1.bNewInputType) -> (ReturnValue:bool))
   > SetHeightOverride(self=Get SizeBox_Mobile() -> (SizeBox_Mobile:Size Box Object Reference), InHeightOverride=80)
   ClearHeightOverride(execute=@n2.else, self:Size Box Object Reference=Get SizeBox_Mobile())
   Event PreConstruct()
   > Set Data!(execute, self, InText:Text=Get Text_Details() -> (Text_Details:Text)) -> (then) {is=title K2Node_CallFunction}
   > SetVisuals!(execute, self, MaxWidth:int=Get Max Width() -> (Max Width:int=0)) -> (then)
   Event Construct()
   > Bind Event to On Input Method Changed(self=Get CommonInputSubsystem(), Delegate=@n1)
   > OnInputMethodChanged!(execute, self, bNewInputType:ECommonInputType Enum=GetCurrentInputType(self=Get CommonInputSubsystem())) -> (then) {is=title K2Node_CallFunction}
   Event Destruct()
   > Unbind Event from On Input Method Changed(self=Get CommonInputSubsystem(), Delegate=@n1)
  ]}
 Set Data {nodes [
   n1 = Set Data!() -> (then, InText:Text) {is=title K2Node_FunctionEntry}
   > n2 = Set Text_Details(Text_Details:Text=@n1) -> (Output_Get:Text)
   > "SetText (Text)"(self=Get Text_Description() -> (Text_Description:Common Text Object Reference), InText=@n2)
  ]}
 SetVisuals {nodes [
   n1 = SetVisuals!() -> (then, MaxWidth:int) {is=title K2Node_FunctionEntry}
   > SetWrapTextWidth!(
    execute
    self:Common Text Object Reference=[Get Text_Description(), Get Text_Title() -> (Text_Title:Common Text Object Reference)]
    InWrapTextAt:int=@n1
   ) -> (then)
  ]}
]
variables [
 Max Width {type=integer, default=640}
 ParentalControlsScreenClass {type=Class, default=&/ParentalControlsScreen.ParentalControlsScreen_C}
 Text_Details {type=string, default="[PH] Text to be set in C++ in nested widget"}
]
widgets {
 info {inheritedWidgetCount=0, parentClass=&CommonActivatableWidget, rootWidgetClass=&VerticalBox, widgetCount=5}
 widgets [
  VerticalBox_25 {widgetClassPath=&VerticalBox, widget=VerticalBox_25}
  Text_Title {
   widgetClassPath=&CommonTextBlock
   bIsVariable=true
   parent=VerticalBox_25
   widget=Text_Title
   properties {
    justification=Center
    lineHeightPercentage=0.85
    text=Account Setup Required
    textTransformPolicy=ToUpper
    font {
     bForceMonospaced=false
     bMaterialIsStencil=false
     fontMaterial=null
     fontObject=&Burbank
     letterSpacing=0
     monospacedWidth=1
     size=42
     skewAmount=0
     typefaceFontName=Burbank-Black
     outlineSettings {
      bApplyOutlineToDropShadows=false
      bMiteredCorners=false
      bSeparateFillAlpha=false
      outlineMaterial=null
      outlineSize=0
      outlineColor=<rgba 0 0 0 1>
     }
    }
    shadowOffset=<0 0>
   }
   slot {class=VerticalBoxSlot, properties {padding=<ltrb 0 0 0 16>}}
  }
  Text_Description {
   widgetClassPath=&CommonTextBlock
   bIsVariable=true
   parent=VerticalBox_25
   widget=Text_Description
   properties {
    justification=Center
    mobileFontSizeMultiplier=1.2
    text="[PH] Text to be set in C++ in nested widget"
    font {
     bForceMonospaced=false
     bMaterialIsStencil=false
     fontMaterial=null
     fontObject=&BurbankSmall
     letterSpacing=0
     monospacedWidth=1
     size=18
     skewAmount=0
     typefaceFontName=Bold
     outlineSettings {
      bApplyOutlineToDropShadows=false
      bMiteredCorners=false
      bSeparateFillAlpha=false
      outlineMaterial=null
      outlineSize=0
      outlineColor=<rgba 0 0 0 1>
     }
    }
    shadowOffset=<0 0>
   }
   slot {class=VerticalBoxSlot}
  }
  SizeBox_Mobile {
   widgetClassPath=&SizeBox
   bIsVariable=true
   parent=VerticalBox_25
   widget=SizeBox_Mobile
   properties {heightOverride=80}
   slot {class=VerticalBoxSlot, properties {padding=<ltrb 0 32 0 0>}}
  }
  WBP_UIKit_Button_Loud {
   widgetClassPath=/FortUILibrary/UIKit/Buttons/WBP_UIKit_Button_Loud
   bIsVariable=true
   parent=SizeBox_Mobile
   widget=WBP_UIKit_Button_Loud
   properties {bHideInputAction=true, text=Open Parental Controls}
   slot {class=SizeBoxSlot}
  }
 ]
}
